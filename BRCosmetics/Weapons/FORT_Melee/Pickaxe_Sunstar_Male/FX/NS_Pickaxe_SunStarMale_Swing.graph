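# Niagara swing effect for the male Sunstar pickaxe, combining ribbon trails with sparkle and star sprites.
# Three anim-trail ribbons follow weapon bones or sockets, fading with alpha scaled by time of day.
# CPU sparkles spawn on the user-supplied skeletal mesh, drift with curl noise, flicker and cycle colors.
# GPU stars get vortex velocity, fall under gravity in random blue-to-orange tints, and collide, except in the Stars_Mobile copy.
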NiagaraSystem /BRCosmetics/Weapons/FORT_Melee/Pickaxe_Sunstar_Male/FX/NS_Pickaxe_SunStarMale_Swing
dependencies {
 usedDataInterfaces [&NiagaraDataInterfaceCurve, &NiagaraDataInterfaceColorCurve, &NiagaraDataInterfaceArrayColor]
 usedDynamicInputs []
 usedModules [&SystemState]
}
emitters {
 Sparkles {
  emitterUpdateScript [
   &EmitterState(Life Cycle Mode=::NewEnumerator1, Loop Behavior=::NewEnumerator1, Loop Delay=0.07, Loop Duration=1, UseLoopDelay=true)
   &SpawnRate(SpawnRate=&RandomRangeFloat(Minimum=70, Maximum=170), Use Spawn Probability=false)
  ]
  particleSpawnScript [
   &SkeletalMeshLocation(
    Bone / Socket Index=2
    MaskSpawnGroup=false
    Mesh Sampling Type=::NewEnumerator5
    Skeletal Mesh=$User.SkeletalMesh
    Socket Sampling Mode=::NewEnumerator5
   )
   &InitializeParticle(
    Color Mode=::NewEnumerator0
    Lifetime Max=1.25
    Lifetime Min=0.3
    Lifetime Mode=::NewEnumerator1
    Position Mode=::NewEnumerator0
    Sprite Size Mode=::NewEnumerator1
    Uniform Sprite Size Max=25
    Uniform Sprite Size Min=13
    Write Material Random=true
   ) {module=InitializeParticle001}
   &ShapeLocation(MaskSpawnGroup=false, Override Seed=false, Sphere Radius=22, UseOwnerScale=true)
   &AddVelocity(Velocity=&RandomRangeVector(Minimum=<2 2 20>, Maximum=<-2 -2 5>))
  ]
  particleUpdateScript [
   &ParticleState()
   &CurlNoiseForce(Noise Frequency=20.6872, Noise Strength=666, Pan Noise=false)
   &Drag(
    Drag=&FloatFromCurve(
     FloatCurve=DataInterface() {propertyValues {Curve {keys [{interpMode=RCIM_Cubic, tangentMode=RCTM_User, time=0, value=0}, {interpMode=RCIM_Cubic, tangentMode=RCTM_User, time=1, value=1}]}}}
     Scale Curve=2
    )
    Use Linear Drag=true
    Use Rotational Drag=false
   )
   &AccelerationForce(Acceleration=&RandomRangeVector(Minimum=<-111 0 -600>, Maximum=<55 0 -600>))
   &SolveForcesAndVelocity(Clamp Velocity=false, Limit Acceleration=false)
   &ScaleSpriteSize(Uniform Curve Sprite Scale=DataInterface() {propertyValues {Curve {keys [
        <at 0 1 RCIM_Cubic>
        {arriveTangent=-0.1852, interpMode=RCIM_Cubic, leaveTangent=-0.1852, time=0.14, value=0.29}
        {arriveTangent=-0.6296, interpMode=RCIM_Cubic, leaveTangent=-0.6296, time=0.27, value=0.95}
        {arriveTangent=0.0333, interpMode=RCIM_Cubic, leaveTangent=0.0333, time=0.41, value=0.12}
        {arriveTangent=0.0333, interpMode=RCIM_Cubic, leaveTangent=0.0333, time=0.57, value=0.96}
        {arriveTangent=-0.1333, interpMode=RCIM_Cubic, leaveTangent=-0.1333, time=0.71, value=0.13}
        {arriveTangent=0.2414, interpMode=RCIM_Cubic, leaveTangent=0.2414, time=0.87, value=0.92}
        <at 1 0.2 RCIM_Cubic>
       ]}}})
   &Color(Color=&ColorFromCurve(ColorCurve=DataInterface() {propertyValues {
       AlphaCurve {keys [<at 0 1 RCIM_Cubic>,<at 1 0 RCIM_Cubic>]}
       BlueCurve {keys [<at 0.1693 1 RCIM_Cubic>,<at 0.2323 0>,<at 0.3084 0.974>,<at 0.4498 0.974>,<at 0.4629 0>,<at 0.5735 0>,<at 0.5782 1>,<at 0.7447 1>,<at 0.7792 0>,<at 1 0 RCIM_Cubic>]}
       GreenCurve {keys [<at 0.1693 0.9002 RCIM_Cubic>,<at 0.2323 1>,<at 0.3084 0>,<at 0.4498 0>,<at 0.4629 0.5346>,<at 0.5735 1>,<at 0.5782 1>,<at 0.7447 1>,<at 0.7792 0.5346>,<at 1 0.5346 RCIM_Cubic>]}
       RedCurve {keys [<at 0.1693 0 RCIM_Cubic>,<at 0.2323 1>,<at 0.3084 0.7258>,<at 0.4498 0.7258>,<at 0.4629 1>,<at 0.5735 1>,<at 0.5782 0>,<at 0.7447 0>,<at 0.7792 1>,<at 1 1 RCIM_Cubic>]}
      }})) {module=Color001}
  ]
  rendererClasses [&NiagaraSpriteRendererProperties]
  renderers [{rendererClass=&NiagaraSpriteRendererProperties}]
 }
 Stars {
  simTarget=GPUComputeSim
  emitterUpdateScript [
   &EmitterState(Life Cycle Mode=::NewEnumerator1, Loop Behavior=::NewEnumerator1, Loop Delay=0.07, Loop Duration=1, UseLoopDelay=true)
   &SpawnRate(SpawnRate=&RandomRangeFloat(Minimum=30, Maximum=40), Use Spawn Probability=false)
  ]
  particleSpawnScript [
   &SkeletalMeshLocation(MaskSpawnGroup=false, Mesh Sampling Type=::NewEnumerator5, OverrideRandomSeed=false, Skeletal Mesh=$User.SkeletalMesh)
   &InitializeParticle(
    Color Mode=::NewEnumerator0
    Lifetime Max=1.25
    Lifetime Min=0.7
    Lifetime Mode=::NewEnumerator1
    Position Mode=::NewEnumerator0
    Sprite Rotation Mode=::NewEnumerator2
    Sprite Size Mode=::NewEnumerator1
    Uniform Sprite Size Min=4
    Write Material Random=true
   )
   &ShapeLocation(MaskSpawnGroup=false, Override Seed=false, Sphere Radius=22, UseOwnerScale=true)
   /Niagara/Modules/Update/Velocity/VortexVelocity(
    Influence Falloff Radius=400
    Module.Use Influence Falloff=true
    Velocity Amount:NiagaraFloat=&RandomRangeFloat(Minimum=33, Maximum=500)
    Vortex Axis:Vector3f=<0 0 0.5>
   )
   &ApplyInitialForces()
  ]
  particleUpdateScript [
   &ParticleState()
   &AccelerationForce(Acceleration=<0 0 -800>) {module=AccelerationForce001}
   &Collision(
    Age Colliding Particles=false
    GPU Collision Type=::NewEnumerator2
    Module.Control Rest State Via Surface Normal=false
    Module.EnableMaxCollisionCount=false
    Module.Randomize Collision Normal Vector=false
   )
   &SolveForcesAndVelocity(Clamp Velocity=false, Limit Acceleration=false)
   &SolveRotationalForcesAndVelocity()
   &Color(Color=&SelectLinearColorFromArray(
     Array Sampling Mode=::NewEnumerator0
     Direct Array Index=?
     Interpolation=&RandomRangeFloat(Minimum=0, Maximum=1)
     Color Selection Array=DataInterface() {propertyValues {ColorData [<rgba 0.0874 0.3854 1 1>, <rgba 1 0.5346 0 1>]}}
     Random Seed=0
    )) {module=Color002}
   &ScaleSpriteSize(Uniform Curve Sprite Scale=DataInterface() {propertyValues {Curve {keys [
        <at 0 0 RCIM_Cubic>
        {arriveTangent=9.2727, interpMode=RCIM_Cubic, leaveTangent=9.2727, time=0.07, value=1.17}
        {arriveTangent=-0.5417, interpMode=RCIM_Cubic, leaveTangent=-0.5417, time=0.11, value=1.02}
        {arriveTangent=-0.1897, interpMode=RCIM_Cubic, leaveTangent=-0.1897, time=0.55, value=0.91}
        {arriveTangent=-2.0222, interpMode=RCIM_Cubic, leaveTangent=-2.0222, time=0.69, value=0.91}
        <at 1 0 RCIM_Cubic>
       ]}}})
  ]
  rendererClasses [&NiagaraSpriteRendererProperties]
  renderers [{rendererClass=&NiagaraSpriteRendererProperties}]
 }
 Stars_Mobile {
  simTarget=GPUComputeSim
  emitterUpdateScript [
   &EmitterState(Life Cycle Mode=::NewEnumerator1, Loop Behavior=::NewEnumerator1, Loop Delay=0.07, Loop Duration=1, UseLoopDelay=true)
   &SpawnRate(SpawnRate=&RandomRangeFloat(Minimum=30, Maximum=40), Use Spawn Probability=false)
  ]
  particleSpawnScript [
   &SkeletalMeshLocation(MaskSpawnGroup=false, Mesh Sampling Type=::NewEnumerator5, OverrideRandomSeed=false, Skeletal Mesh=$User.SkeletalMesh)
   &InitializeParticle(
    Color Mode=::NewEnumerator0
    Lifetime Max=1.25
    Lifetime Min=0.7
    Lifetime Mode=::NewEnumerator1
    Position Mode=::NewEnumerator0
    Sprite Rotation Mode=::NewEnumerator2
    Sprite Size Mode=::NewEnumerator1
    Uniform Sprite Size Min=4
    Write Material Random=true
   )
   &ShapeLocation(MaskSpawnGroup=false, Override Seed=false, Sphere Radius=22, UseOwnerScale=true)
   &VortexVelocity(
    Influence Falloff Radius=400
    Module.Use Influence Falloff=true
    Velocity Amount:NiagaraFloat=&RandomRangeFloat(Minimum=33, Maximum=500)
    Vortex Axis:Vector3f=<0 0 0.5>
   )
   &ApplyInitialForces()
  ]
  particleUpdateScript [
   &ParticleState()
   &AccelerationForce(Acceleration=<0 0 -800>) {module=AccelerationForce001}
   &SolveForcesAndVelocity(Clamp Velocity=false, Limit Acceleration=false)
   &SolveRotationalForcesAndVelocity()
   &Color(Color=&SelectLinearColorFromArray(
     Array Sampling Mode=::NewEnumerator0
     Direct Array Index=?
     Interpolation=&RandomRangeFloat(Minimum=0, Maximum=1)
     Color Selection Array=DataInterface() {propertyValues {ColorData [<rgba 0.0874 0.3854 1 1>, <rgba 1 0.5346 0 1>]}}
     Random Seed=0
    )) {module=Color002}
   &ScaleSpriteSize(Uniform Curve Sprite Scale=DataInterface() {propertyValues {Curve {keys [
        <at 0 0 RCIM_Cubic>
        {arriveTangent=9.2727, interpMode=RCIM_Cubic, leaveTangent=9.2727, time=0.07, value=1.17}
        {arriveTangent=-0.5417, interpMode=RCIM_Cubic, leaveTangent=-0.5417, time=0.11, value=1.02}
        {arriveTangent=-0.1897, interpMode=RCIM_Cubic, leaveTangent=-0.1897, time=0.55, value=0.91}
        {arriveTangent=-2.0222, interpMode=RCIM_Cubic, leaveTangent=-2.0222, time=0.69, value=0.91}
        <at 1 0 RCIM_Cubic>
       ]}}})
  ]
  rendererClasses [&NiagaraSpriteRendererProperties]
  renderers [{rendererClass=&NiagaraSpriteRendererProperties}]
 }
 Trail_Lg_Wispy {
  emitterUpdateScript [
   &EmitterState(Life Cycle Mode=::NewEnumerator1, Loop Behavior=::NewEnumerator1, Loop Delay=1, Loop Duration=1, UseLoopDelay=true)
   &FN_SpawnAnimTrail_v2(AnimTrailWidthMode=::FromFirst)
  ]
  particleSpawnScript [
   /Game/Effects/Niagara/Modules/AnimTrail/FN_InitializeAnimTrail/FN_InitializeAnimTrailBoneIndecies(Second Bone or Socket Index=1)
   /Game/Effects/Niagara/Modules/AnimTrail/FN_InitializeAnimTrail/FN_InitializeAnimTrail(AnimTrailWidth=1, Lifetime=0.6)
  ]
  particleUpdateScript [
   &ScaleRibbonWidth(Ribbon Width Scale=&FloatFromCurve(FloatCurve=DataInterface() {propertyValues {Curve {keys [
         {arriveTangent=-1.2579, interpMode=RCIM_Cubic, time=0, value=1}
         {arriveTangent=-0.9664, interpMode=RCIM_Cubic, leaveTangent=-0.9664, tangentMode=RCTM_User, time=1, value=0}
        ]}}}))
   &DynamicMaterialParameters(
    Index 0 Param 1=&FloatFromCurve(FloatCurve=DataInterface() {propertyValues {Curve {keys [
         {arriveTangent=-2.8011, interpMode=RCIM_Cubic, time=0, value=1}
         {arriveTangent=-1, interpMode=RCIM_Cubic, leaveTangent=-1, time=0.4919, value=0.0513}
         <at 1 0>
        ]}}})
    Index 0 Param 3=-0.8
    Index 0 Param 4=&FloatFromCurve(FloatCurve=DataInterface() {propertyValues {Curve {keys [{arriveTangent=-2.7093, interpMode=RCIM_Cubic, time=0, value=1}, <at 1 0>]}}})
    Module.Param0WriteEnabled=true
    Module.Param1WriteEnabled=true
    Module.Param2WriteEnabled=true
    Module.Param3WriteEnabled=true
   )
   &AccelerationForce(Acceleration=<-222 0 0>)
   &SolveForcesAndVelocity(Clamp Velocity=false, Limit Acceleration=false)
   &FN_UpdateAnimTrail_v2()
   &Color(
    Color=&ColorFromCurve(ColorCurve=DataInterface() {propertyValues {
       AlphaCurve {keys [<at 0 1 RCIM_Cubic>,<at 1 0 RCIM_Cubic>]}
       BlueCurve {keys [<at 0 1 RCIM_Cubic>,<at 0.1039 1>,<at 0.237 0.9725>,<at 0.2727 0.9725>,<at 0.3214 0>,<at 1 0 RCIM_Cubic>]}
       GreenCurve {keys [<at 0 0.7929 RCIM_Cubic>,<at 0.1039 0.7929>,<at 0.237 0.2157>,<at 0.2727 0.2157>,<at 0.3214 0.5346>,<at 1 0.5346 RCIM_Cubic>]}
       RedCurve {keys [<at 0 0 RCIM_Cubic>,<at 0.1039 0>,<at 0.237 0.9608>,<at 0.2727 0.9608>,<at 0.3214 1>,<at 1 1 RCIM_Cubic>]}
      }})
    Scale Alpha=&Multiply_Float(
     A=&FloatFromCurve(
      FloatCurve=DataInterface() {propertyValues {Curve {keys [{interpMode=RCIM_Cubic, tangentMode=RCTM_User, time=0, value=1}, {interpMode=RCIM_Cubic, tangentMode=RCTM_User, time=1, value=0}]}}}
      Scale Curve=0.1
     )
     B=&FNTimeOfDayScalar(Night Scale=0.3, Morning Scale=0.5, Day Scale=0.5, Evening Scale=0.2)
    )
   ) {module=Color001}
  ]
  rendererClasses [&NiagaraRibbonRendererProperties]
  renderers [{rendererClass=&NiagaraRibbonRendererProperties}]
 }
 Trail_Sm_Solid {
  emitterUpdateScript [
   &EmitterState(Life Cycle Mode=::NewEnumerator1, Loop Behavior=::NewEnumerator1, Loop Duration=0.2, UseLoopDelay=false)
   &FN_SpawnAnimTrail_v2(AnimTrailWidthMode=::FromFirst)
  ]
  particleSpawnScript [
   &FN_InitializeAnimTrailBoneIndecies(Second Bone or Socket Index=1)
   &FN_InitializeAnimTrail(AnimTrailWidth=1, Lifetime=0.3)
  ]
  particleUpdateScript [
   &ScaleRibbonWidth(Ribbon Width Scale=&FloatFromCurve(
     FloatCurve=DataInterface() {propertyValues {Curve {keys [
         {arriveTangent=-1.6158, interpMode=RCIM_Cubic, leaveTangent=-1.6158, tangentMode=RCTM_User, time=0, value=1}
         {arriveTangent=-0.9664, interpMode=RCIM_Cubic, leaveTangent=-0.9664, tangentMode=RCTM_User, time=1, value=0}
        ]}}}
     Scale Curve=0.5
    ))
   &FN_UpdateAnimTrail_v2()
   &Color(
    Color=<rgba 0 0.7843 1 1>
    Scale Alpha=&Multiply_Float(
     A=&FloatFromCurve(
      FloatCurve=DataInterface() {propertyValues {Curve {keys [{interpMode=RCIM_Cubic, tangentMode=RCTM_User, time=0, value=1}, {interpMode=RCIM_Cubic, tangentMode=RCTM_User, time=1, value=0}]}}}
      Scale Curve=0.5
     )
     B=&FNTimeOfDayScalar(Night Scale=0.15, Morning Scale=0.09, Day Scale=0.2, Evening Scale=0.1)
    )
   ) {module=Color001}
  ]
  rendererClasses [&NiagaraRibbonRendererProperties]
  renderers [{rendererClass=&NiagaraRibbonRendererProperties}]
 }
 Trail_Sm_Wispy {
  emitterUpdateScript [
   &EmitterState(Life Cycle Mode=::NewEnumerator1, Loop Behavior=::NewEnumerator1, Loop Delay=2, Loop Duration=0.2, UseLoopDelay=false)
   &FN_SpawnAnimTrail_v2(AnimTrailWidthMode=::FromFirst)
  ]
  particleSpawnScript [
   &FN_InitializeAnimTrailBoneIndecies(First Bone or Socket Index=1, Second Bone or Socket Index=3)
   &FN_InitializeAnimTrail(AnimTrailWidth=1, Lifetime=0.3)
  ]
  particleUpdateScript [
   &ScaleRibbonWidth(Ribbon Width Scale=&FloatFromCurve(FloatCurve=DataInterface() {propertyValues {Curve {keys [
         {arriveTangent=-1.2579, interpMode=RCIM_Cubic, time=0, value=1}
         {arriveTangent=-0.9664, interpMode=RCIM_Cubic, leaveTangent=-0.9664, tangentMode=RCTM_User, time=1, value=0}
        ]}}}))
   &AccelerationForce(Acceleration=<-222 0 0>)
   &SolveForcesAndVelocity(Clamp Velocity=false, Limit Acceleration=false)
   &FN_UpdateAnimTrail_v2()
   &DynamicMaterialParameters(
    Index 0 Param 1=&FloatFromCurve(FloatCurve=DataInterface() {propertyValues {Curve {keys [
         {arriveTangent=-2.8011, interpMode=RCIM_Cubic, time=0, value=1}
         {arriveTangent=-1, interpMode=RCIM_Cubic, leaveTangent=-1, time=0.4919, value=0.0513}
         <at 1 0>
        ]}}})
    Index 0 Param 3=-0.8
    Index 0 Param 4=&FloatFromCurve(FloatCurve=DataInterface() {propertyValues {Curve {keys [{arriveTangent=-2.7093, interpMode=RCIM_Cubic, time=0, value=1}, <at 1 0>]}}})
    Module.Param0WriteEnabled=true
    Module.Param1WriteEnabled=true
    Module.Param2WriteEnabled=true
    Module.Param3WriteEnabled=true
   )
   &Color(
    Color=<rgba 0 0.7843 1 1>
    Scale Alpha=&Multiply_Float(
     A=&FloatFromCurve(
      FloatCurve=DataInterface() {propertyValues {Curve {keys [{interpMode=RCIM_Cubic, tangentMode=RCTM_User, time=0, value=1}, {interpMode=RCIM_Cubic, tangentMode=RCTM_User, time=1, value=0}]}}}
      Scale Curve=0.5
     )
     B=&FNTimeOfDayScalar(Night Scale=0.1, Morning Scale=0.1, Day Scale=0.2, Evening Scale=0.07)
    )
   ) {module=Color001}
   &CameraOffset(Camera Offset Amount=2)
  ]
  rendererClasses [&NiagaraRibbonRendererProperties]
  renderers [{rendererClass=&NiagaraRibbonRendererProperties}]
 }
}
summary {
 systemName=NS_Pickaxe_SunStarMale_Swing
 emitters [
  {emitterName=Trail_Sm_Wispy, rendererClasses [&NiagaraRibbonRendererProperties]}
  {emitterName=Trail_Sm_Solid, rendererClasses [&NiagaraRibbonRendererProperties]}
  {emitterName=Sparkles, rendererClasses [&NiagaraSpriteRendererProperties]}
  {emitterName=Stars, simTarget=GPUComputeSim, rendererClasses [&NiagaraSpriteRendererProperties]}
  {emitterName=Trail_Lg_Wispy, rendererClasses [&NiagaraRibbonRendererProperties]}
  {emitterName=Stars_Mobile, simTarget=GPUComputeSim, rendererClasses [&NiagaraSpriteRendererProperties]}
 ]
 userVariables [
  User.RibbonIDOffset {type=NiagaraInt32}
  User.SkeletalMesh {
   type=NiagaraDataInterfaceSkeletalMesh
   defaultValue {
    struct=NiagaraExt_VariableValue_DataInterface
    value {dataInterface=&self:NiagaraDataInterfaceSkeletalMesh_0}
   }
  }
 ]
}
