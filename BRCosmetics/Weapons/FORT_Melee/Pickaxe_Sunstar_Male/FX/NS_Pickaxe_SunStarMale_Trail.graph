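# Niagara ribbon trail for the Sunstar Male pickaxe, built on the shared AnimTrail modules from a skeletal mesh user parameter and bone or socket indices.
# A single Trail emitter spawns and updates short-lived ribbon particles that narrow over their lifetime.
# Color jumps sharply from purple to pale cream, then whitens, while alpha fades and is scaled by time of day, dimmer at night.
# A duplicate Color module without that scaling is disabled.
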NiagaraSystem /BRCosmetics/Weapons/FORT_Melee/Pickaxe_Sunstar_Male/FX/NS_Pickaxe_SunStarMale_Trail
dependencies {
 usedDataInterfaces [&NiagaraDataInterfaceCurve, &NiagaraDataInterfaceColorCurve]
 usedDynamicInputs []
 usedModules [&SystemState]
}
emitters {Trail {
  emitterUpdateScript [
   &EmitterState(Life Cycle Mode=::NewEnumerator1, Loop Behavior=::NewEnumerator1, Loop Duration=0.2, UseLoopDelay=false)
   &FN_SpawnAnimTrail_v2(AnimTrailWidthMode=::FromFirst)
  ]
  particleSpawnScript [
   /Game/Effects/Niagara/Modules/AnimTrail/FN_InitializeAnimTrail/FN_InitializeAnimTrailBoneIndecies(First Bone or Socket Index=1)
   /Game/Effects/Niagara/Modules/AnimTrail/FN_InitializeAnimTrail/FN_InitializeAnimTrail(AnimTrailWidth=0.95, Lifetime=0.3)
  ]
  particleUpdateScript [
   &ScaleRibbonWidth(Ribbon Width Scale=&FloatFromCurve(
     FloatCurve=DataInterface() {propertyValues {Curve {keys [
         {arriveTangent=-1.2579, interpMode=RCIM_Cubic, time=0, value=1}
         {arriveTangent=-0.9664, interpMode=RCIM_Cubic, leaveTangent=-0.9664, tangentMode=RCTM_User, time=1, value=0}
        ]}}}
     Scale Curve=1.2
    ))
   &FN_UpdateAnimTrail_v2()
   &Color(
    Color=&ColorFromCurve(ColorCurve=DataInterface() {propertyValues {
       AlphaCurve {keys [<at 0 1 RCIM_Cubic>,<at 1 0 RCIM_Cubic>]}
       BlueCurve {keys [<at 0.3476 0.5393 RCIM_Cubic>,<at 0.3809 0.8086>,<at 0.7316 0.6303>,<at 1 1 RCIM_Cubic>]}
       GreenCurve {keys [<at 0.3476 0 RCIM_Cubic>,<at 0.3809 0.9109>,<at 0.7316 0.8279>,<at 1 1 RCIM_Cubic>]}
       RedCurve {keys [<at 0.3476 0.4019 RCIM_Cubic>,<at 0.3809 1>,<at 0.7316 1>,<at 1 1 RCIM_Cubic>]}
      }})
    Scale Alpha=&FloatFromCurve(
     FloatCurve=DataInterface() {propertyValues {Curve {keys [{interpMode=RCIM_Cubic, tangentMode=RCTM_User, time=0, value=1}, {interpMode=RCIM_Cubic, tangentMode=RCTM_User, time=1, value=0}]}}}
     Scale Curve=1.7
    )
   ) {enabled=false, module=Color001}
   &Color(
    Color=&ColorFromCurve(ColorCurve=DataInterface() {propertyValues {
       AlphaCurve {keys [<at 0 1 RCIM_Cubic>,<at 1 0 RCIM_Cubic>]}
       BlueCurve {keys [<at 0.3476 0.5393 RCIM_Cubic>,<at 0.3809 0.8086>,<at 0.7316 0.6303>,<at 1 1 RCIM_Cubic>]}
       GreenCurve {keys [<at 0.3476 0 RCIM_Cubic>,<at 0.3809 0.9109>,<at 0.7316 0.8279>,<at 1 1 RCIM_Cubic>]}
       RedCurve {keys [<at 0.3476 0.4019 RCIM_Cubic>,<at 0.3809 1>,<at 0.7316 1>,<at 1 1 RCIM_Cubic>]}
      }})
    Scale Alpha=&Multiply_Float(
     A=&FloatFromCurve(
      FloatCurve=DataInterface() {propertyValues {Curve {keys [{interpMode=RCIM_Cubic, tangentMode=RCTM_User, time=0, value=1}, {interpMode=RCIM_Cubic, tangentMode=RCTM_User, time=1, value=0}]}}}
      Scale Curve=1.7
     )
     B=&FNTimeOfDayScalar(Night Scale=0.6, Morning Scale=1, Evening Scale=1)
    )
   )
  ]
  rendererClasses [&NiagaraRibbonRendererProperties]
  renderers [{rendererClass=&NiagaraRibbonRendererProperties}]
 }}
summary {
 systemName=NS_Pickaxe_SunStarMale_Trail
 emitters [{emitterName=Trail, rendererClasses [&NiagaraRibbonRendererProperties]}]
 userVariables [
  User.RibbonIDOffset {type=NiagaraInt32}
  User.SkeletalMesh {
   type=NiagaraDataInterfaceSkeletalMesh
   defaultValue {
    struct=NiagaraExt_VariableValue_DataInterface
    value {dataInterface=&self:NiagaraDataInterfaceSkeletalMesh_2}
   }
  }
 ]
}
